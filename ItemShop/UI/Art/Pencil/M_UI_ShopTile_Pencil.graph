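# UI material for the Pencil item shop tile, drawing slowly panning diamond patterns over a two-color gradient with parameterized colors.
# Two particle grids tile a distance-field diamond texture with anti-aliased edges: one staggers alternate columns, the other is offset and faded by a linear gradient.
# Gradient-colored and secondary-color diamonds, a rotated diagonal fade and a radial center-focus mask are shared through named reroutes for layered lerp and mask compositing.
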Material /ItemShop/UI/Art/Pencil/M_UI_ShopTile_Pencil
graph {
 properties {bUsedWithStaticMesh=false, materialDomain=MD_UI, pixelDepthOffsetMode=PDOM_Legacy, usedShadingModels=Default Lit}
 nodes [
  Constant() {r=0.5}
  n1 = NamedRerouteUsage()
  n2 = NamedRerouteUsage()
  n3 = NamedRerouteUsage()
  n4 = Max(A=NamedRerouteUsage(), B=NamedRerouteUsage())
  n5 = OneMinus(None=NamedRerouteUsage())
  n6 = NamedRerouteUsage()
  LinearInterpolate(
   A=NamedRerouteUsage()
   B=LinearInterpolate(A=LinearInterpolate(A=@n1, B=@n2, Alpha=@n3), B=NamedRerouteUsage(), Alpha=NamedRerouteUsage())
   Alpha=Min(A=Min(A=@n4, B=@n5), B=OneMinus(None=@n6))
  )
  n7 = VectorParameter() {parameterName=SecondaryDiamonds, defaultValue=<rgba 0.8802 0.7075 0 1>}
  LinearInterpolate(
   A=@n7.RGB
   Alpha=MaterialFunctionCall(GradientStart=Constant() {r=0.3}, GradientEnd=Constant() {r=0.75}, RotationAngle=Constant() {r=0.35}) {materialFunction=&MF_UI_LinearGradient}.Result
  )
  LinearInterpolate(
   A=VectorParameter() {parameterName=BG Upper Gradient, defaultValue=<rgba 0.4963 0.3279 1 1>}.RGB
   B=VectorParameter() {parameterName=SecondaryDiamonds, defaultValue=<rgba 0.8802 0.7075 0 1>}.RGB
  ) {constAlpha=0.75}
  Multiply(A=NamedRerouteUsage(), B=NamedRerouteUsage())
  n8 = Panner(Coordinate=Multiply(A=NamedRerouteUsage()) {constB=0.5}, Time=Time(), Speed=Constant2Vector() {g=0.025, r=0.015})
  n9 = ComponentMask(None=@n8) {r=true}
  n10 = Constant() {r=8}
  n11 = Saturate(None=MaterialFunctionCall!(Value=Multiply(A=@n9, B=@n10)) {materialFunction=/FortUILibrary/MaterialFunctions/Basics/MF_UI_EvenOrOdd}.Result)
  n12 = Constant()
  n13 = MaterialFunctionCall!(
   UV=@n8
   Global Offset=AppendVector(A=Constant(), B=LinearInterpolate(Alpha=@n11) {constB=0.25})
   Global Rotation
   Tiling=@n10
   Cell Offset=@n12
   Cell Scale=LinearInterpolate(Alpha=NamedRerouteUsage()) {constA=0.25}
   Cell Rotation=@n12
   Random Seed
  ) {materialFunction=/FortUILibrary/MaterialFunctions/Basics/MF_UI_ParticleGrid}
  NamedRerouteDeclaration(None=@n13.Random Rows) {name=Rows, nodeColor=<rgba 0 1 0.5547 1>}
  NamedRerouteDeclaration(None=@n13.Random Cell Values) {name=RandomCell, nodeColor=<rgba 0 0.7422 1 1>}
  n14 = Constant()
  n15 = MaterialFunctionCall(
   UV=@n8
   Global Offset=Constant2Vector() {g=0.11, r=0.11}
   Tiling=@n10
   Cell Offset=@n14
   Cell Scale=LinearInterpolate(Alpha=NamedRerouteUsage()) {constA=0.25}
   Cell Rotation=@n14
  ) {materialFunction=&MF_UI_ParticleGrid}
  NamedRerouteDeclaration(None=@n15.Random Rows) {name=Rows_1, nodeColor=<rgba 0 1 0.5547 1>}
  NamedRerouteDeclaration(None=@n15.Random Cell Values) {name=RandomCell_1, nodeColor=<rgba 0 0.7422 1 1>}
  n16 = Constant() {r=0.3}
  n17 = Constant() {r=0.75}
  n18 = Constant() {r=0.35}
  NamedRerouteDeclaration(None=LinearInterpolate(
    A=VectorParameter() {parameterName=BG Lower Gradient, defaultValue=<rgba 0.1428 0.6875 0.4616 1>}.RGB
    B=VectorParameter() {parameterName=BG Upper Gradient, defaultValue=<rgba 0.4963 0.3279 1 1>}.RGB
    Alpha=MaterialFunctionCall(GradientStart=@n16, GradientEnd=@n17, RotationAngle=@n18) {materialFunction=&MF_UI_LinearGradient}.Result
   )) {name=BackgroundGradient, nodeColor=<rgba 1 0.3281 0 1>}
  NamedRerouteDeclaration(None=@n13.Random Columns) {name=Columns, nodeColor=<rgba 1 0 0.6094 1>}
  NamedRerouteDeclaration(None=MaterialFunctionCall() {materialFunction=&MF_UI_UVScaling}.Scaled to Fit) {name=UVScaling, nodeColor=<rgba 0.2891 1 0 1>}
  NamedRerouteDeclaration(None=@n7.RGB) {name=SecondaryDiamondsColor, nodeColor=<rgba 0 1 0.4844 1>}
  n19 = TextureSample(UVs=@n13.Cells UV) {samplerType=SAMPLERTYPE_LinearColor, texture=/OfferCatalog/Art/Pencil/T_UI_Pencil_Diamond}
  NamedRerouteDeclaration(None=MaterialFunctionCall(
    DistanceField=MaterialFunctionCall!(RGBA Channels=@n19.R, Glow Min, Glow Max, Outline Thickness) {materialFunction=/FortUILibrary/MaterialFunctions/Basics/MF_UI_ChannelPackedDistanceField}.DF
    Threshold=Constant() {r=0.4}
   ) {materialFunction=&MF_UI_AntiAliasing}.Outer) {name=diamonds alpha, nodeColor=<rgba 0.3594 1 0 1>}
  n20 = Constant() {r=0.1}
  n21 = Constant() {r=0.75}
  n22 = Constant() {r=0.9}
  NamedRerouteDeclaration(None=LinearInterpolate(
    A=VectorParameter() {parameterName=Diamonds Lower Gradient, defaultValue=<rgba 0.2956 0.7237 1 1>}.RGB
    B=VectorParameter() {parameterName=Diamonds Upper Gradient, defaultValue=<rgba 0.4138 0.1684 1 1>}.RGB
    Alpha=MaterialFunctionCall(GradientStart=@n20, GradientEnd=@n21, RotationAngle=@n22) {materialFunction=&MF_UI_LinearGradient}.Result
   )) {name=diamonds color, nodeColor=<rgba 0.2422 1 0 1>}
  n23 = MaterialFunctionCall(
   DistanceField=MaterialFunctionCall(RGBA Channels=TextureSample(UVs=@n15.Cells UV) {samplerType=SAMPLERTYPE_LinearColor, texture=&T_UI_Pencil_Diamond}.R) {materialFunction=&MF_UI_ChannelPackedDistanceField}.DF
   Threshold=Constant() {r=0.4}
  ) {materialFunction=&MF_UI_AntiAliasing}
  n24 = MaterialFunctionCall(GradientStart=Constant() {r=0.25}, GradientEnd=Constant() {r=0.65}, RotationAngle=Constant() {r=0.35}) {materialFunction=&MF_UI_LinearGradient}
  NamedRerouteDeclaration(None=Multiply(A=Multiply(A=@n24.Result, B=@n23.Outer)) {constB=0.8}) {name=fading diamonds alpha, nodeColor=<rgba 0 1 0.4375 1>}
  n25 = MaterialFunctionCall(UVs=NamedRerouteUsage(), Rotation Center=Constant() {r=0.5}, "Rotation Angle (0-1)"=Constant() {r=0.6}) {materialFunction=&CustomRotator}
  n26 = ComponentMask(None=Abs(None=ConstantBiasScale(None=@n25.Rotated Values) {bias=-0.5, scale=2})) {g=true}
  n27 = ScalarParameter() {defaultValue=0.75, parameterName=Subtract}
  n28 = Saturate(None=Multiply(A=Subtract(A=@n26, B=@n27), B=ScalarParameter() {defaultValue=0.65, parameterName=Falloff}))
  n29 = ScalarParameter() {defaultValue=2, parameterName=Intensity}
  NamedRerouteDeclaration(None=Saturate(None=Multiply(A=@n28, B=@n29))) {name=diagonal fade, nodeColor=<rgba 0 0.4844 1 1>}
  n30 = MaterialFunctionCall(
   UVs=Multiply(A=MaterialFunctionCall() {materialFunction=&MF_UI_UVScaling}.ScaleX, B=Constant2Vector() {g=1.5, r=1})
   Texture Scale=Constant() {r=1.25}
  ) {materialFunction=&ScaleUVsByCenter}
  n31 = Constant() {r=0.25}
  n32 = Constant() {r=1}
  NamedRerouteDeclaration(None=Saturate(None=MaterialFunctionCall(UV=@n30.UVs, GradientStart=@n31, GradientEnd=@n32) {materialFunction=&MF_UI_RadialGradient}.Smooth)) {name=center focus, nodeColor=<rgba 0 1 0.0391 1>}
 ]
}
